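# Translucent, lit cloud material for Niagara mesh particles in the Stunning Mountain glider effects, with internal lightning flashes.
# Panned smoke textures with per-particle random offsets build the noisy color, while a named reroute holds a Fresnel center-to-edge tint.
# Smoke and Fresnel masks shape emissive lightning, flickered by noise and a time sine and driven by particle dynamic parameters.
# Scrolling masks, view-angle fade and particle alpha form dithered opacity, and noise along vertex normals builds world position offset.
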Material /Game/Gadgets/Assets/VinderTech_GliderChute/Glider_StunningMountain/FX/Material/M_Glider_StunningMountain_Cloud
graph {
 properties {
  bUsedWithNiagaraMeshParticles=true
  blendMode=BLEND_Translucent
  opacityMaskClipValue=0.1
  pixelDepthOffsetMode=PDOM_Legacy
  usedShadingModels=Default Lit
 }
 nodes [
  n1 = ParticleColor()
  n2 = TextureCoordinate() {vTiling=0.7}
  n3 = Power(Base=DotProduct(
    A=Constant3Vector() {constant=<rgba 0 0 1 0>}
    B=TextureSample(UVs=@n2) {samplerType=SAMPLERTYPE_Normal, texture=/Game/Global/Textures/Default/Tileable_Patterns/1Normal}.RGB
   )) {constExponent=5}
  n4 = TextureCoordinate()
  n5 = Add(A=@n4, B=Constant2Vector() {g=0.5, r=0.5})
  n6 = AppendVector(A=Constant(), B=ParticleRandom())
  n7 = TextureSample(UVs=Panner(Coordinate=Add(A=@n5, B=@n6)) {speedY=0.7}) {texture=&T_Inky_Smoke_Tile}
  n8 = AppendVector(A=Constant(), B=ParticleRandom())
  n9 = TextureSample(UVs=Panner(Coordinate=Add(A=@n4, B=@n8)) {speedY=0.5}) {texture=&T_Inky_Smoke_Tile}
  n10 = Multiply(A=VectorParameter() {parameterName=ColorNoise, defaultValue=<rgba 0.0646 0.1324 0.4 1>}.RGB, B=Saturate(None=Add(A=@n7.R, B=@n9.R)))
  n11 = Add(A=Add(A=Multiply(A=@n1.RGB) {constB=0.01}, B=Multiply(A=@n3, B=@n10) {constA=1}), B=Add(A=NamedRerouteUsage()) {constB=0})
  n12 = TextureSample() {samplerType=SAMPLERTYPE_Grayscale, texture=&LowResBlurredNoise}
  n13 = LinearInterpolate(A=Multiply(A=Multiply(A=@n1.RGB, B=@n12.R), B=@n11), B=@n11, Alpha=ScalarParameter() {defaultValue=1, parameterName=Use_AddBaseTexture})
  LinearInterpolate(A=@n13, B=Multiply(A=@n13, B=ParticleColor().RGB), Alpha=ScalarParameter() {parameterName=LerpParticleColor})
  n14 = TextureCoordinate()
  n15 = Panner(Coordinate=Add(A=Add(A=@n14) {constB=0}) {constB=0.38})
  n16 = TextureCoordinate()
  n17 = AppendVector(A=Constant(), B=ParticleRandom())
  n18 = TextureSample(UVs=Panner(Coordinate=Add(A=@n16, B=@n17)) {speedY=0.5}) {texture=&T_Inky_Smoke_Tile}
  n19 = Power(Base=Multiply(A=TextureSample(UVs=@n15) {texture=&T_Inky_Smoke_Tile}.R, B=@n18.R)) {constExponent=1.6}
  n20 = Multiply(
   A=MaterialFunctionCall(Invert Fresnel=StaticBool() {value=true}, Power=LinearInterpolate(Alpha=@n18.R) {constA=3, constB=20}) {materialFunction=&Fresnel_Function}.Result
   B=MaterialFunctionCall(Power=LinearInterpolate(Alpha=@n18.R) {constA=0.5}) {materialFunction=&Fresnel_Function}.Result
  )
  n21 = Multiply(
   A=Multiply(A=VectorParameter() {parameterName=Color Lighting, defaultValue=<rgba 156.2501 662.9366 1500 1>}.RGB, B=ParticleRandom())
   B=Saturate(None=Multiply(A=@n19, B=@n20))
  )
  n22 = Add(
   A=TextureCoordinate() {uTiling=0, vTiling=0}
   B=DynamicParameter() {paramNames [WPO Intensity, BrightOffset, LightningFlash, Param4]}.BrightOffset
  )
  n23 = LinearInterpolate(Alpha=TextureSample(UVs=Panner(Coordinate=@n22) {speedX=0.07, speedY=0.07}) {texture=/Game/Global/Textures/Default/Noise/Noise}.R) {constA=-1, constB=3}
  n24 = Constant() {r=1}
  n25 = Multiply(
   A=LinearInterpolate(
    B=Saturate(None=@n23)
    Alpha=MaterialFunctionCall(Speed=@n24) {materialFunction=&DebugTimeSine}.Result
   )
   B=DynamicParameter() {paramNames [WPO Intensity, BrightOffset, LightningFlash, Param4]}.LightningFlash
  )
  MaterialFunctionCall(Emissive=Add(A=Multiply(A=@n21, B=@n25), B=VectorParameter() {parameterName=BaseEmissiveColor, defaultValue=<rgba 0 0 0 1>}.RGB)) {materialFunction=&MF_FX_EmissiveAdjustments}
  n26 = Add(A=TextureCoordinate() {uTiling=2}, B=Constant2Vector() {g=-0.06})
  n27 = ComponentMask(None=TextureCoordinate()) {g=true}
  n28 = Multiply(A=Power(Base=Sine(None=@n27) {period=2}) {constExponent=1.7}) {constB=5}
  n29 = LinearInterpolate(
   A=TextureSample(UVs=Panner(Coordinate=@n26) {speedX=-1}) {samplerSource=SSM_Wrap_WorldGroupSettings, texture=/Game/Effects/Fort_Effects/Textures/T_Smoke_Ring_Mask2}.R
   B=LinearInterpolate(Alpha=Saturate(None=@n28))
   Alpha=ScalarParameter() {defaultValue=1, parameterName=Use_SecondMask}
  )
  n30 = Constant()
  n31 = ParticleRandom()
  n32 = Panner(Coordinate=Add(A=TextureCoordinate(), B=AppendVector(A=@n30, B=@n31))) {speedY=0.7}
  n33 = Power(
   Base=MaterialFunctionCall("FadeSensitivity (optional)"=ScalarParameter() {defaultValue=2, parameterName=Edge Fade}) {materialFunction=&FadeBasedOnViewAngle}.Result
   Exp=LinearInterpolate(Alpha=TextureSample(UVs=@n32) {texture=&T_Inky_Smoke_Tile}.R) {constA=3, constB=10}
  )
  n34 = Multiply(A=Saturate(None=Multiply(B=@n33) {constA=10}))
  n35 = Clamp(None=Multiply(A=ParticleColor().A, B=Multiply(A=@n34, B=@n29))) {maxDefault=0.999}
  MaterialFunctionCall(Alpha Threshold=@n35) {materialFunction=&DitherTemporalAA_Fortnite}
  MaterialFunctionCall(Number=@n35) {materialFunction=&DebugScalarValues}
  n36 = Add(A=TextureCoordinate(), B=ParticleRandom())
  n37 = Multiply(
   A=TextureSample(UVs=Panner(Coordinate=@n36) {speedX=0.5, speedY=1}) {samplerType=SAMPLERTYPE_Grayscale, texture=&LowResBlurredNoise}.R
   B=Multiply(
    A=ScalarParameter() {defaultValue=15, parameterName=WPO Intensity}
    B=DynamicParameter() {paramNames [WPO Intensity, BrightOffset, LightningFlash, Param4]}.WPO Intensity
   )
  ) {constB=25}
  n38 = VertexNormalWS()
  NamedRerouteDeclaration(None=Multiply(A=Multiply(A=@n37, B=@n38) {constB=0})) {name=WPO, nodeColor=<rgba 1 0 0.7734 1>}
  n39 = MaterialFunctionCall(Power=Constant() {r=1.5}) {materialFunction=&Fresnel_Function}
  NamedRerouteDeclaration(None=LinearInterpolate(
    A=VectorParameter() {parameterName=ColorCenter, defaultValue=<rgba 0.0394 0.0429 0.0573 1>}.RGB
    B=VectorParameter() {parameterName=ColorOutside, defaultValue=<rgba 0.3542 0.5156 1 1>}.RGB
    Alpha=Multiply(A=@n39.Result)
   )) {name=Fresnel, nodeColor=<rgba 0 0.0625 0.5208 1>}
  Reroute(None=NamedRerouteUsage())
 ]
}
