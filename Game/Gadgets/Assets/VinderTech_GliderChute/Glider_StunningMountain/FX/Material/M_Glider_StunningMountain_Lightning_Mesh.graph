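# Translucent, unlit, two-sided Niagara mesh-particle material for the StunningMountain glider's lightning effect.
# A shared clock, selectable between continuous, stepped or particle-keyed time, scrolls the bolt texture and flickers dissolve UVs between offsets.
# Emissive layers channel tints, tail brightness and optional galaxy detail; opacity combines dissolve, optional offset-sample edge, tail and depth fades.
# Two blended noises, optionally quantized, displace vertices along a local axis, scaled by dynamic WPOStr.
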Material /Game/Gadgets/Assets/VinderTech_GliderChute/Glider_StunningMountain/FX/Material/M_Glider_StunningMountain_Lightning_Mesh
graph {
 properties {
  bUsedWithNiagaraMeshParticles=true
  blendMode=BLEND_Translucent
  opacityMaskClipValue=0.1
  pixelDepthOffsetMode=PDOM_Legacy
  shadingModel=MSM_Unlit
  twoSided=true
  usedShadingModels=Unlit
 }
 nodes [
  n1 = Time() {bOverride_Period=true, period=6000}
  n2 = DynamicParameter() {paramNames [TimeRandom,WPOStr,KeyPanning,Offset]}
  n3 = ScalarParameter() {parameterName=UseKeyPanningAnim}
  n4 = Time()
  n5 = Multiply(A=Ceil(None=Multiply(A=@n1) {constB=8})) {constB=0.5}
  n6 = Add(
   A=LinearInterpolate(
    A=LinearInterpolate(A=@n1, B=@n2.KeyPanning, Alpha=@n3)
    B=LinearInterpolate(A=@n4, B=@n5) {constAlpha=0.325}
    Alpha=ScalarParameter() {parameterName=Use Step anim}
   )
   B=@n2.TimeRandom
  )
  n7 = Add(A=@n6)
  n8 = VectorParameter() {parameterName=Tex_UVSpeed, defaultValue=<rgba -1 0 0 1>}
  n9 = AppendVector(A=@n8.R, B=@n8.R)
  n10 = VectorParameter() {parameterName=Tex_UVMul, defaultValue=<rgba 1 1 0 1>}
  n11 = Multiply(A=Add(A=TextureCoordinate(), B=Multiply(A=@n7, B=@n9)), B=AppendVector(A=@n10.R, B=@n10.R))
  n12 = TextureSampleParameter2D(UVs=@n11) {
   parameterName=Tex_Default
   texture=/Game/Gadgets/Assets/VinderTech_GliderChute/Glider_StunningMountain/FX/T_Glider_StunningMountain_Lightning
  }
  n13 = Multiply(A=VectorParameter() {parameterName=BlueChannelEdgeColor, defaultValue=<rgba 0 1.5729 9 1>}.RGB, B=@n12.B)
  n14 = TextureSampleParameter2D(UVs=@n11) {
   parameterName=Tex_Default
   texture=&T_Glider_StunningMountain_Lightning
  }
  n15 = Multiply(A=VectorParameter() {parameterName=CenterColor, defaultValue=<rgba 0 0.3495 2 1>}.RGB, B=@n14.G)
  n16 = TextureCoordinate()
  n17 = VectorParameter() {group=YiMod, parameterName=Dissolve_UVMulAndOffset, defaultValue=<rgba 1 1 0 0.1>}
  n18 = AppendVector(A=@n17.R, B=@n17.R)
  n19 = ScalarParameter() {defaultValue=-2, group=YiMod, parameterName=Dissolve_FlickSpeed}
  n20 = Add(
   A=Add(A=Multiply(A=@n16, B=@n18), B=AppendVector(A=@n17.B, B=@n17.B))
   B=LinearInterpolate(
    A=Constant2Vector() {r=0.2}
    B=Constant2Vector() {r=0.5}
    Alpha=MaterialFunctionCall(Time=@n6, Speed=@n19) {materialFunction=&TimeWithSpeedVariable}.Frac Time
   )
  )
  n21 = TextureSampleParameter2D(UVs=@n20) {
   group=YiMod
   parameterName=Dissolve_Texture
   samplerType=SAMPLERTYPE_LinearGrayscale
   texture=/Game/Effects/Fort_Effects/Emotes/UltraLight/Texture/T_PlasmaCoils
  }
  n22 = ScalarParameter() {parameterName=InvertDissolveTexture}
  n23 = Power(Base=LinearInterpolate(A=@n21.R, B=OneMinus(None=@n21.R), Alpha=@n22)) {constExponent=1}
  n24 = TextureCoordinate()
  n25 = LinearInterpolate(
   A=ScalarParameter() {defaultValue=-0.9, parameterName=Dissolve mask Min, sortPriority=3}
   B=ScalarParameter() {defaultValue=1, parameterName=Dissolve mask Max, sortPriority=4}
   Alpha=Power(Base=ComponentMask(None=@n24) {r=true}) {constExponent=1.3}
  )
  n26 = Saturate(None=LinearInterpolate(Alpha=Power(Base=@n23, Exp=@n25)) {constA=-2, constB=2})
  n27 = Constant()
  n28 = Add(B=ScalarParameter() {defaultValue=-0.1, group=YiMod, parameterName=CenterTex_DissolveVOffset})
  n29 = TextureSampleParameter2D(UVs=Add(A=AppendVector(A=@n27, B=@n28), B=@n20)) {
   group=YiMod
   parameterName=Dissolve_Texture
   samplerType=SAMPLERTYPE_LinearGrayscale
   texture=&T_PlasmaCoils
  }
  n30 = OneMinus(None=@n29.R)
  n31 = Power(Base=Power(Base=LinearInterpolate(A=@n29.R, B=@n30, Alpha=@n22)) {constExponent=1}, Exp=@n25)
  n32 = Subtract(A=@n26, B=Saturate(None=LinearInterpolate(Alpha=@n31) {constA=-2, constB=2}))
  n33 = ScalarParameter() {defaultValue=5, group=YiMod, parameterName=CenterTex_DissolveMaskScale}
  n34 = OneMinus(None=Saturate(None=Multiply(A=@n32, B=@n33)))
  n35 = LinearInterpolate(B=@n34, Alpha=ScalarParameter() {group=YiMod, parameterName=EnableCenterTexDissolveEdge}) {constA=1}
  n36 = LinearInterpolate(B=Add(A=@n13, B=Multiply(A=@n15, B=@n35)), Alpha=ScalarParameter() {parameterName=Use Center Tex})
  n37 = Power(Base=ComponentMask(None=TextureCoordinate()) {r=true}) {constExponent=6}
  n38 = ScalarParameter() {defaultValue=0.3, parameterName=Noise Lerp Min, sortPriority=1}
  n39 = ScalarParameter() {defaultValue=1, parameterName=Noise Lerp Max, sortPriority=2}
  n40 = TextureSample(UVs=@n11) {samplerType=SAMPLERTYPE_Grayscale, texture=&LowResBlurredNoise}
  n41 = TextureObject() {texture=/Game/Characters/Player/Female/Medium/Bodies/F_MED_SuperNovaTaro/FX/Material/T_SuperNovaTaro_Skin_Nebula}
  n42 = MaterialFunctionCall!(CameraVector, Texture=@n41, Tiling=ScalarParameter() {defaultValue=10, parameterName=GalaxyUVTile}, Offset, AxisFadeContrast) {materialFunction=/Engine/Functions/Engine_MaterialFunctions01/Texturing/TriplanarCameraVector}
  n43 = ScalarParameter() {defaultValue=1, parameterName=GalaxyTexturePower}
  n44 = VectorParameter() {
   group=NewGalaxySample
   parameterName=NewGalaxySample_UVControl
   channelNames {a=VPan, b=UPan, g=VScale, r=UScale}
   defaultValue=<rgba 2 2 0 0>
  }
  n45 = MaterialFunctionCall(Texture=@n41, UVControl_RGB=@n44, UVControl_A=@n44) {materialFunction=&FXCommon_TextureSample}
  n46 = Multiply(
   A=VectorParameter() {parameterName=GalaxyDetailColorMulti, defaultValue=<rgba 0 1.6213 5 1>}.RGB
   B=LinearInterpolate(A=Power(Base=@n42.Result, Exp=@n43), B=Power(Base=@n45.RGB, Exp=@n43), Alpha=ScalarParameter() {parameterName=UseNewGalaxySample})
  )
  n47 = Panner(Coordinate=TextureCoordinate(), Time=@n6) {speedX=-1}
  n48 = Saturate(None=LinearInterpolate(Alpha=TextureSample(UVs=@n47) {samplerType=SAMPLERTYPE_Grayscale, texture=&LowResBlurredNoise}.R) {constB=0.7})
  n49 = Multiply(
   A=LinearInterpolate(B=LinearInterpolate(Alpha=@n37) {constA=1, constB=10}, Alpha=ScalarParameter() {defaultValue=1, parameterName=Use Tail Brightness}) {constA=1}
   B=LinearInterpolate(
    A=LinearInterpolate(A=@n38, B=@n39, Alpha=@n40.R) {constA=0.3}
    B=LinearInterpolate(A=@n46, B=@n48) {constAlpha=0.9}
    Alpha=ScalarParameter() {parameterName=Use Center Galaxy Details}
   )
  ) {constA=1}
  MaterialFunctionCall(Emissive=Multiply(A=ParticleColor().RGB, B=Add(A=@n36, B=@n49))) {materialFunction=&MF_FX_EmissiveAdjustments}
  n50 = VectorParameter() {
   group=EdgeOpacityMask
   parameterName=EdgeOpacityMask_EdgeMask_InvertControl
   channelNames {a="", b=Power, g=Multiply, r=Offset}
   defaultValue=<rgba 0 1 1 1>
  }
  n51 = VectorParameter() {
   group=EdgeOpacityMask
   parameterName=EdgeOpacityMask_EdgeMask_Control
   channelNames {a="", b=Power, g=Multiply, r=Offset}
   defaultValue=<rgba 0 1 1 1>
  }
  n52 = MaterialFunctionCall!(
   UorV=MaterialFunctionCall(Float2=TextureCoordinate()) {materialFunction=&BreakOutFloat2Components}.R
   Invert_EdgeControl=@n50
   NoInvert_EdgeControl=@n51
  ) {materialFunction=/Game/Effects/Niagara/MasterMaterials/Functions/Common/FXCommon_EdgeMask_Ex}
  n53 = ScalarParameter() {parameterName=EnableEdgeOpacityMask}
  n54 = TextureSampleParameter2D(UVs=@n11) {
   parameterName=Tex_Default
   texture=&T_Glider_StunningMountain_Lightning
  }
  n55 = Add(
   A=LinearInterpolate(B=Desaturation(None=@n13), Alpha=ScalarParameter() {parameterName=Use Center Tex})
   B=StaticComponentMaskParameter(None=@n54.RGBA) {defaultR=true, parameterName=RGBMask}
  )
  n56 = TextureCoordinate()
  n57 = Power(Base=OneMinus(None=ComponentMask(None=@n56) {r=true}), Exp=ScalarParameter() {defaultValue=1, parameterName=Tail Power})
  n58 = ScalarParameter() {defaultValue=5, parameterName=Tail Intensity}
  n59 = Multiply(A=@n26, B=Saturate(None=Multiply(A=@n57, B=@n58) {constB=5}))
  n60 = Multiply(
   A=Multiply(A=ParticleColor().A, B=Multiply(A=@n55, B=@n59))
   B=LinearInterpolate(B=@n34, Alpha=ScalarParameter() {group=YiMod, parameterName=EnableOpacityDissolveEdge}) {constA=1}
  )
  n61 = Power(
   Base=DepthFade(Opacity=Constant() {r=1}, FadeDistance=ScalarParameter() {group=YiMod, parameterName=Depthfade})
   Exp=ScalarParameter() {defaultValue=2, group=YiMod, parameterName=DepthfadePow}
  )
  n62 = TextureCoordinate()
  n63 = Add(A=OneMinus(None=ComponentMask(None=@n62) {r=true})) {constB=-0.05}
  Multiply(
   A=Multiply(A=LinearInterpolate(B=@n52.Mask, Alpha=@n53) {constA=1}, B=Multiply(A=@n60, B=@n61))
   B=Saturate(None=Multiply(A=@n63) {constB=5})
  )
  n64 = VectorParameter() {group=YiMod, parameterName=WPO_A_UVControl, channelNames {a=VPan, b=UPan, g=VScale, r=UScale}, defaultValue=<rgba 1 1 -1 0>}
  n65 = MaterialFunctionCall!(UVInput=TextureCoordinate() {uTiling=0.7, vTiling=0}, TimeInput, Control_RGB=@n64, Control_A=@n64) {materialFunction=/Game/Effects/Niagara/MasterMaterials/Functions/Common/FXCommon_QuickUVControl}
  n66 = VectorParameter() {group=YiMod, parameterName=WPO_B_UVControl, channelNames {a=VPan, b=UPan, g=VScale, r=UScale}, defaultValue=<rgba 1 1 -0.5 0>}
  n67 = MaterialFunctionCall(UVInput=TextureCoordinate() {uTiling=0.5, vTiling=0}, Control_RGB=@n66, Control_A=@n66) {materialFunction=&FXCommon_QuickUVControl}
  n68 = Add(
   A=LinearInterpolate(
    A=TextureSample(UVs=@n65.UV) {texture=/Game/Weapons/FORT_Melee/Pickaxe_MadameMoth/FX/T_Fibre_Noise}.R
    B=TextureSample(UVs=@n67.UV) {texture=/Game/Weapons/WeaponSkins/Textures/T_Swirly_Noise_D1}.R
    Alpha=ScalarParameter() {defaultValue=0.5, parameterName=WPO_TextureBlend}
   )
   B=ScalarParameter() {group=YiMod, parameterName=WPOOffset}
  )
  n69 = ScalarParameter() {defaultValue=1, parameterName=WPO_LevelCount}
  n70 = Multiply(A=@n68, B=@n69)
  n71 = LinearInterpolate(A=@n68, B=Divide(A=Round(None=@n70), B=@n69), Alpha=ScalarParameter() {parameterName=WPO_UseLevel})
  n72 = VectorParameter() {parameterName=WPO_MaskInfor, channelNames {a="", b=Highest, g=Lowest, r=Power}, defaultValue=<rgba 1 0.4 1 1>}
  n73 = Multiply(A=ComponentMask(None=TextureCoordinate()) {r=true}) {constB=4}
  n74 = LinearInterpolate(A=@n72.Lowest, B=@n72.Lowest, Alpha=Power(Base=Saturate(None=@n73), Exp=@n72.Power)) {constA=0.4}
  Multiply(
   A=Transform(None=VectorParameter() {defaultValue=<rgba 0 17 0 1>}.RGB) {transformSourceType=TRANSFORMSOURCE_Local}
   B=Multiply(A=Multiply(A=@n71, B=@n74), B=DynamicParameter() {paramNames [TimeRandom,WPOStr,KeyPanning,Offset]}.WPOStr)
  )
 ]
}
